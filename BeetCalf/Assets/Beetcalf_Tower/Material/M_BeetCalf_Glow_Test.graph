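# Translucent unlit glow test material in the BeetCalf tower assets.
# A remapped, powered normal-to-camera falloff is multiplied by a scene-versus-pixel depth fade to soften intersections.
# Separate branches build a green color compensated by inverse eye adaptation, an inward normal offset and an inverted Fresnel.
# No links from these branches to material outputs appear, and a standalone EyeAdaptation node feeds nothing.
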Material /BeetCalf/Assets/Beetcalf_Tower/Material/M_BeetCalf_Glow_Test
graph {
 properties {bAutomaticallySetUsageInEditor=false, blendMode=BLEND_Translucent, shadingModel=MSM_Unlit, usedShadingModels=Unlit}
 nodes [
  EyeAdaptation()
  Multiply(A=Multiply(A=VertexNormalWS()) {constB=-1}, B=ScalarParameter() {defaultValue=20, parameterName=Offset Strength})
  n1 = Subtract(A=Subtract(A=SceneDepth().R, B=PixelDepth().R), B=ScalarParameter() {defaultValue=50, parameterName=Depth Fade Offset})
  n2 = ScalarParameter() {defaultValue=100, parameterName=Depth Fade Distance}
  n3 = VertexNormalWS()
  n4 = CameraVectorWS()
  n5 = InverseLinearInterpolate(A=ScalarParameter(), B=ScalarParameter() {defaultValue=1, parameterName=Param_1}, Value=ConstantBiasScale(None=DotProduct(A=@n3, B=@n4)))
  Multiply(A=Power(Base=Saturate(None=@n5), Exp=ScalarParameter() {defaultValue=2, parameterName=Param_2}), B=Saturate(None=Divide(A=@n1, B=@n2))) {constA=1}
  Multiply(A=Constant3Vector() {constant=<rgba 0.1568 1 0.0966 1>}, B=EyeAdaptationInverse().EyeAdaptationInverse)
  OneMinus(None=Fresnel(
    ExponentIn=ScalarParameter() {defaultValue=0.1, parameterName=Fresnel Exponent}
    BaseReflectFractionIn=ScalarParameter() {defaultValue=0.9999, parameterName=Fresnel Reflect Fration}
   ) {baseReflectFraction=1, exponent=1})
 ]
}
