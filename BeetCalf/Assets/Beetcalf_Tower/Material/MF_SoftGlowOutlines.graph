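# MF_SoftGlowOutlines is a material function that outputs emissive, opacity and normal-based vertex offset for soft, translucent glow outlines around meshes.
# Custom HLSL builds the outline by radially sampling scene or custom depth with noise jitter, then subtracting the center mask.
# A static-switch Fresnel alternative uses a depth mask.
# Parameters drive a world-space sine pulse and camera-distance thickness blending, and a quality switch dims opacity on lower settings.
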MaterialFunction /BeetCalf/Assets/Beetcalf_Tower/Material/MF_SoftGlowOutlines
graph {
 properties {previewBlendMode=BLEND_Translucent}
 nodes [
  n1 = NamedRerouteUsage()
  n2 = Multiply(A=Constant() {r=0.1}, B=NamedRerouteUsage())
  n3 = SetMaterialAttributes!(
   MaterialAttributes
   EmissiveColor=NamedRerouteUsage()
   Opacity=QualitySwitch(Default=@n1, Low=@n2, Medium=LinearInterpolate(A=@n1, B=@n2) {constAlpha=0.7})
   WorldPositionOffset=NamedRerouteUsage()
  ) {attributeSetTypes [#1, #2, #3]}
  FunctionOutput(None=@n3)
  n4 = ScalarParameter() {defaultValue=1, group=SoftGlow, parameterName=GlowPower, sortPriority=13}
  LinearInterpolate(A=ScalarParameter() {defaultValue=1, group=GlowPulse, parameterName=Glow Pulse Min Power Adjustment}, B=@n4, Alpha=NamedRerouteUsage())
  n5 = ScalarParameter() {defaultValue=1, group=SoftGlow, parameterName=GlowStrength, sortPriority=14}
  LinearInterpolate(A=ScalarParameter() {defaultValue=1, group=GlowPulse, parameterName=Glow Pulse Min Strength Adjustment}, B=@n5, Alpha=NamedRerouteUsage())
  n6 = ScalarParameter() {defaultValue=0.05, group=GlowPulse, parameterName=Glow Pulse Thickness Far Adjustment}
  n7 = ScalarParameter() {defaultValue=0.05, group=SoftGlow, parameterName=GlowThickness_Far, sortPriority=12}
  n8 = NamedRerouteUsage()
  n9 = NamedRerouteDeclaration(None=LinearInterpolate(
    A=LinearInterpolate(A=@n6, B=@n7, Alpha=@n8)
    B=ScalarParameter() {defaultValue=0.5, group=SoftGlow, parameterName=GlowThickness, sortPriority=11}
    Alpha=NamedRerouteUsage()
   )) {name=OutlineThickness, nodeColor=<rgba 0.1719 0 1 1>}
  n10 = Constant() {r=30}
  MaterialFunctionCall!(
   Vector2=NamedRerouteDeclaration(None=Divide(A=@n9, B=@n10)) {name=Distance, nodeColor=<rgba 0.8984 1 0 1>}
   MaximumNumberOfDigits
   UVs
   DebugTextLocation RG_UpperRight BA_LowerLeft
   Component Spacing
  ) {materialFunction=/Engine/Functions/Engine_MaterialFunctions02/Utility/DebugFloat2Values}
  n11 = ScalarParameter() {defaultValue=1, parameterName=FresnelMethod_Exp, sortPriority=40}
  n12 = Power(
   Base=OneMinus(None=Fresnel(ExponentIn=@n11) {baseReflectFraction=0, exponent=1})
   Exp=ScalarParameter() {defaultValue=5, group=SoftGlow, parameterName=FresnelMethod_Power, sortPriority=40}
  ) {constExponent=3}
  n13 = NamedRerouteUsage()
  n14 = ScalarParameter() {defaultValue=0.5, group=SoftGlow, parameterName=FresnelMethod_DepthMaskPower, sortPriority=40}
  NamedRerouteDeclaration(None=LinearInterpolate(A=Saturate(None=@n12), Alpha=Power(Base=@n13, Exp=@n14) {constExponent=0.5}) {constB=0}) {name=FresnelMethod_Opacity, nodeColor=<rgba 1 0 0.5156 1>}
  MaterialExpressionNamedRerouteUsage_2 = NamedRerouteUsage()
  MaterialExpressionNamedRerouteUsage_4 = NamedRerouteUsage()
  MaterialExpressionNamedRerouteUsage_8 = NamedRerouteUsage()
  MaterialExpressionNamedRerouteUsage_38 = NamedRerouteUsage()
  MaterialExpressionNamedRerouteUsage_39 = NamedRerouteUsage()
  MaterialExpressionNamedRerouteUsage_11 = NamedRerouteUsage()
  MaterialExpressionNamedRerouteUsage_40 = NamedRerouteUsage()
  MaterialExpressionNamedRerouteUsage_41 = NamedRerouteUsage()
  MaterialExpressionNamedRerouteUsage_23 = NamedRerouteUsage()
  MaterialExpressionNamedRerouteUsage_13 = NamedRerouteUsage()
  MaterialExpressionNamedRerouteUsage_24 = NamedRerouteUsage()
  n15 = Custom!(
   Distance=@MaterialExpressionNamedRerouteUsage_2
   DistanceSteps=@MaterialExpressionNamedRerouteUsage_4
   RadialSteps=@MaterialExpressionNamedRerouteUsage_8
   TempAARotation=@MaterialExpressionNamedRerouteUsage_38
   DistanceMask=@MaterialExpressionNamedRerouteUsage_39
   Tex=@MaterialExpressionNamedRerouteUsage_11
   TempAADistance=@MaterialExpressionNamedRerouteUsage_40
   RadialOffset=@MaterialExpressionNamedRerouteUsage_41
   DivideF=@MaterialExpressionNamedRerouteUsage_23
   PDepth=@MaterialExpressionNamedRerouteUsage_13
   PixelDepthOffset=@MaterialExpressionNamedRerouteUsage_24
  ) {
   code="float CurColor = 0;\r\nfloat2 BaseUV = MaterialFloat2(ScreenAlignedPosition(Parameters.ScreenPosition).xy);\r\nfloat2 NewUV = BaseUV;\r\nfloat StepSize = Distance / (int) DistanceSteps;\r\nfloat CurDistance = 0;\r\nfloat2 CurOffset = 0;\r\nfloat TwoPi = 6.283185;\r\nfloat Substep = 0;\r\nfloat2 ScenePixels=View.BufferSizeAndInvSize.xy*BaseUV;\r\nScenePixels+=View.TemporalAAParams.r;\r\nfloat2 RandomSamp = ((uint)(ScenePixels.x) + 2 * (uint)(ScenePixels.y)) % 5;\r\nRandomSamp+=Texture2DSample(Tex,TexSampler,ScenePixels);\r\nRandomSamp/=5;\r\nRandomSamp-=0.5;\r\nTempAARotation*=RandomSamp;\r\nTempAADistance*=StepSize*RandomSamp;\r\nfloat Ratio = View.BufferSizeAndInvSize.x / View.BufferSizeAndInvSize.y; //Screen resolution ratio\r\n\r\nint i=0;\r\nif (DistanceSteps<1)\r\n{\r\n    Mask = 0;\r\n    return 0;\r\n}\r\nelse\r\n{\r\n    //CurDistance += 0.5*StepSize;\r\n    while ( i < (int) DistanceSteps)\r\n    {\r\n\r\n        //CurDistance+=StepSize;\r\n        for (int j = 0; j < (int) RadialSteps; j++)\r\n        {\r\n            CurOffset.x = cos(TwoPi*((TempAARotation+Substep) / RadialSteps));\r\n            CurOffset.y = sin(TwoPi*((TempAARotation+Substep) / RadialSteps));\r\n            CurOffset *=DistanceMask;\r\n            NewUV.x = BaseUV.x + (CurOffset.x * (CurDistance+(RandomSamp*TempAADistance)));\r\n            NewUV.y = BaseUV.y + (CurOffset.y * (CurDistance+(RandomSamp*TempAADistance))) * Ratio;\r\n            float tmpdepth = (SceneTextureLookup(NewUV, 13, false)).r;\r\n            float tmpscenedepth = CalcSceneDepth(NewUV);\r\n            float depthcomparemask = ceil(saturate(tmpscenedepth - tmpdepth+0.01))*ceil(saturate(DivideF-(tmpscenedepth-PDepth)+0.01));\r\n            tmpdepth = tmpdepth-PDepth+PixelDepthOffset;\r\n            CurColor += (min(1-saturate(max(0,(tmpdepth))/(DivideF+PixelDepthOffset))-0.99,0)+0.99)*saturate(sign(tmpdepth))*depthcomparemask;\r\n            //CurDistance+=(StepSize+(TempAADistance))/RadialSteps;\r\n            Substep++;\r\n        }\r\n        CurDistance+=StepSize;\r\n        Substep+=RadialOffset;\r\n        i++;\r\n    }\r\n    float tmpdepth2 = (SceneTextureLookup(BaseUV, 13, false)).r-PDepth+PixelDepthOffset;\r\n    float tmpscenedepth2 = CalcSceneDepth(BaseUV);\r\n    float depthcomparemask2 = ceil(saturate(tmpscenedepth2 - tmpdepth2+0.01))*ceil(saturate(DivideF-(tmpscenedepth2-PDepth)+0.01));\r\n    float mask = (min(1-saturate(max(0,(tmpdepth2))/(DivideF+PixelDepthOffset))-0.99,0)+0.99)*saturate(sign(tmpdepth2))*depthcomparemask2;\r\n    Mask = mask;\r\n    CurColor = CurColor / ((int)DistanceSteps*(int)RadialSteps);\r\n    return CurColor;\r\n}"
   desc="\n*** Original source before expression upgrade ***\nfloat3 CurColor;\r\nfloat2 BaseUV = MaterialFloat2(ScreenAlignedPosition(Parameters.ScreenPosition).xy);\r\nfloat2 NewUV = BaseUV;\r\nfloat StepSize = Distance / (int) DistanceSteps;\r\nfloat CurDistance=0;\r\nfloat2 CurOffset;\r\nfloat TwoPi = 6.283185;\r\nfloat Substep;\r\nfloat2 ScenePixels=View.RenderTargetSize*BaseUV;\r\nScenePixels+=View.TemporalAAParams.r;\r\nfloat2 RandomSamp = ((uint)(ScenePixels.x) + 2 * (uint)(ScenePixels.y)) % 5;\r\nRandomSamp+=Texture2DSample(Tex,TexSampler,ScenePixels);\r\nRandomSamp/=5;\r\nRandomSamp-=0.5;\r\nTempAARotation*=RandomSamp;\r\nTempAADistance*=StepSize*RandomSamp;\r\n\r\n\r\nint i=0;\r\nif (DistanceSteps<1)\r\n{\r\nreturn DecodeSceneColorForMaterialNode(NewUV);\r\n}\r\nelse\r\n{\r\n//CurDistance += 0.5*StepSize;\r\nwhile ( i < (int) DistanceSteps)\r\n{\r\n\r\n//CurDistance+=StepSize;\r\nfor (int j = 0; j < (int) RadialSteps; j++)\r\n{\r\nCurOffset.x = cos(TwoPi*((TempAARotation+Substep) / RadialSteps));\r\nCurOffset.y = sin(TwoPi*((TempAARotation+Substep) / RadialSteps));\r\nCurOffset *=DistanceMask;\r\nNewUV.x = BaseUV.x + (CurOffset.x * (CurDistance+(RandomSamp*TempAADistance)));\r\nNewUV.y = BaseUV.y + (CurOffset.y * (CurDistance+(RandomSamp*TempAADistance)));\r\nCurColor += DecodeSceneColorForMaterialNode(NewUV);\r\n//CurDistance+=(StepSize+(TempAADistance))/RadialSteps;\r\nSubstep++;\r\n}\r\nCurDistance+=StepSize;\r\nSubstep+=RadialOffset;\r\ni++;\r\n}\r\nCurColor = CurColor / ((int)DistanceSteps*(int)RadialSteps);\r\nreturn CurColor;\r\n}\n*** Original source before expression upgrade ***\nfloat3 CurColor;\r\nfloat2 BaseUV = MaterialFloat2(ScreenAlignedPosition(Parameters.ScreenPosition).xy);\r\nfloat2 NewUV = BaseUV;\r\nfloat StepSize = Distance / (int) DistanceSteps;\r\nfloat CurDistance=0;\r\nfloat2 CurOffset;\r\nfloat TwoPi = 6.283185;\r\nfloat Substep;\r\nfloat2 ScenePixels=Frame.RenderTargetSize*BaseUV;\r\nScenePixels+=Frame.TemporalAAParams.r;\r\nfloat2 RandomSamp = ((uint)(ScenePixels.x) + 2 * (uint)(ScenePixels.y)) % 5;\r\nRandomSamp+=Texture2DSample(Tex,TexSampler,ScenePixels);\r\nRandomSamp/=5;\r\nRandomSamp-=0.5;\r\nTempAARotation*=RandomSamp;\r\nTempAADistance*=StepSize*RandomSamp;\r\n\r\n\r\nint i=0;\r\nif (DistanceSteps<1)\r\n{\r\nreturn DecodeSceneColorForMaterialNode(NewUV);\r\n}\r\nelse\r\n{\r\n//CurDistance += 0.5*StepSize;\r\nwhile ( i < (int) DistanceSteps)\r\n{\r\n\r\n//CurDistance+=StepSize;\r\nfor (int j = 0; j < (int) RadialSteps; j++)\r\n{\r\nCurOffset.x = cos(TwoPi*((TempAARotation+Substep) / RadialSteps));\r\nCurOffset.y = sin(TwoPi*((TempAARotation+Substep) / RadialSteps));\r\nCurOffset *=DistanceMask;\r\nNewUV.x = BaseUV.x + (CurOffset.x * (CurDistance+(RandomSamp*TempAADistance)));\r\nNewUV.y = BaseUV.y + (CurOffset.y * (CurDistance+(RandomSamp*TempAADistance)));\r\nCurColor += DecodeSceneColorForMaterialNode(NewUV);\r\n//CurDistance+=(StepSize+(TempAADistance))/RadialSteps;\r\nSubstep++;\r\n}\r\nCurDistance+=StepSize;\r\nSubstep+=RadialOffset;\r\ni++;\r\n}\r\nCurColor = CurColor / ((int)DistanceSteps*(int)RadialSteps);\r\nreturn CurColor;\r\n}"
   outputType=CMOT_Float1
   showCode=true
   additionalOutputs [{outputName=Mask, outputType=CMOT_Float1}]
   inputs [
    {inputName=Distance, input {expression=&self:MaterialExpressionNamedRerouteUsage_2}}
    {
     inputName=DistanceSteps
     input {expression=&self:MaterialExpressionNamedRerouteUsage_4}
    }
    {inputName=RadialSteps, input {expression=&self:MaterialExpressionNamedRerouteUsage_8}}
    {
     inputName=TempAARotation
     input {expression=&self:MaterialExpressionNamedRerouteUsage_38}
    }
    {
     inputName=DistanceMask
     input {expression=&self:MaterialExpressionNamedRerouteUsage_39}
    }
    {inputName=Tex, input {expression=&self:MaterialExpressionNamedRerouteUsage_11}}
    {
     inputName=TempAADistance
     input {expression=&self:MaterialExpressionNamedRerouteUsage_40}
    }
    {
     inputName=RadialOffset
     input {expression=&self:MaterialExpressionNamedRerouteUsage_41}
    }
    {inputName=DivideF, input {expression=&self:MaterialExpressionNamedRerouteUsage_23}}
    {inputName=PDepth, input {expression=&self:MaterialExpressionNamedRerouteUsage_13}}
    {
     inputName=PixelDepthOffset
     input {expression=&self:MaterialExpressionNamedRerouteUsage_24}
    }
   ]
  }
  MaterialExpressionNamedRerouteUsage_51 = NamedRerouteUsage()
  MaterialExpressionNamedRerouteUsage_56 = NamedRerouteUsage()
  MaterialExpressionNamedRerouteUsage_57 = NamedRerouteUsage()
  MaterialExpressionNamedRerouteUsage_66 = NamedRerouteUsage()
  MaterialExpressionNamedRerouteUsage_67 = NamedRerouteUsage()
  MaterialExpressionNamedRerouteUsage_60 = NamedRerouteUsage()
  MaterialExpressionNamedRerouteUsage_68 = NamedRerouteUsage()
  MaterialExpressionNamedRerouteUsage_69 = NamedRerouteUsage()
  MaterialExpressionNamedRerouteUsage_61 = NamedRerouteUsage()
  MaterialExpressionNamedRerouteUsage_65 = NamedRerouteUsage()
  MaterialExpressionNamedRerouteUsage_63 = NamedRerouteUsage()
  n16 = Custom(
   Distance=@MaterialExpressionNamedRerouteUsage_51
   DistanceSteps=@MaterialExpressionNamedRerouteUsage_56
   RadialSteps=@MaterialExpressionNamedRerouteUsage_57
   TempAARotation=@MaterialExpressionNamedRerouteUsage_66
   DistanceMask=@MaterialExpressionNamedRerouteUsage_67
   Tex=@MaterialExpressionNamedRerouteUsage_60
   TempAADistance=@MaterialExpressionNamedRerouteUsage_68
   RadialOffset=@MaterialExpressionNamedRerouteUsage_69
   DivideF=@MaterialExpressionNamedRerouteUsage_61
   PDepth=@MaterialExpressionNamedRerouteUsage_65
   PixelDepthOffset=@MaterialExpressionNamedRerouteUsage_63
  ) {
   code="float CurColor = 0;\r\nfloat2 BaseUV = MaterialFloat2(ScreenAlignedPosition(Parameters.ScreenPosition).xy);\r\nfloat2 NewUV = BaseUV;\r\nfloat StepSize = Distance / (int) DistanceSteps;\r\nfloat CurDistance = 0;\r\nfloat2 CurOffset = 0;\r\nfloat TwoPi = 6.283185;\r\nfloat Substep = 0;\r\nfloat2 ScenePixels=View.BufferSizeAndInvSize.xy*BaseUV;\r\nScenePixels+=View.TemporalAAParams.r;\r\nfloat2 RandomSamp = ((uint)(ScenePixels.x) + 2 * (uint)(ScenePixels.y)) % 5;\r\nRandomSamp+=Texture2DSample(Tex,TexSampler,ScenePixels);\r\nRandomSamp/=5;\r\nRandomSamp-=0.5;\r\nTempAARotation*=RandomSamp;\r\nTempAADistance*=StepSize*RandomSamp;\r\nfloat Ratio = View.BufferSizeAndInvSize.x / View.BufferSizeAndInvSize.y; //Screen resolution ratio\r\n\r\nint i=0;\r\nif (DistanceSteps<1)\r\n{\r\n    Mask = 0;\r\n    return 0;\r\n}\r\nelse\r\n{\r\n    //CurDistance += 0.5*StepSize;\r\n    while ( i < (int) DistanceSteps)\r\n    {\r\n        //CurDistance+=StepSize;\r\n        for (int j = 0; j < (int) RadialSteps; j++)\r\n        {\r\n            CurOffset.x = cos(TwoPi*((TempAARotation+Substep) / RadialSteps));\r\n            CurOffset.y = sin(TwoPi*((TempAARotation+Substep) / RadialSteps)) * Ratio;\r\n            CurOffset *=DistanceMask;\r\n            NewUV.x = BaseUV.x + (CurOffset.x * (CurDistance+(RandomSamp*TempAADistance)));\r\n            NewUV.y = BaseUV.y + (CurOffset.y * (CurDistance+(RandomSamp*TempAADistance)));\r\n            float tmpdepth = CalcSceneDepth(NewUV)-PDepth+PixelDepthOffset;\r\n            CurColor += (min(1-saturate(max(0,(tmpdepth))/(DivideF+PixelDepthOffset))-0.99,0)+0.99)*saturate(sign(tmpdepth));\r\n            //CurDistance+=(StepSize+(TempAADistance))/RadialSteps;\r\n            Substep++;\r\n        }\r\n        CurDistance+=StepSize;\r\n        Substep+=RadialOffset;\r\n        i++;\r\n    }\r\n    float tmpdepth2 = CalcSceneDepth(BaseUV)-PDepth+PixelDepthOffset;\r\n    float mask = (min(1-saturate(max(0,(tmpdepth2))/(DivideF+PixelDepthOffset))-0.99,0)+0.99)*saturate(sign(tmpdepth2));\r\n    Mask = mask;\r\n    CurColor = CurColor / ((int)DistanceSteps*(int)RadialSteps);\r\n    return CurColor;\r\n}"
   desc="\n*** Original source before expression upgrade ***\nfloat3 CurColor;\r\nfloat2 BaseUV = MaterialFloat2(ScreenAlignedPosition(Parameters.ScreenPosition).xy);\r\nfloat2 NewUV = BaseUV;\r\nfloat StepSize = Distance / (int) DistanceSteps;\r\nfloat CurDistance=0;\r\nfloat2 CurOffset;\r\nfloat TwoPi = 6.283185;\r\nfloat Substep;\r\nfloat2 ScenePixels=View.RenderTargetSize*BaseUV;\r\nScenePixels+=View.TemporalAAParams.r;\r\nfloat2 RandomSamp = ((uint)(ScenePixels.x) + 2 * (uint)(ScenePixels.y)) % 5;\r\nRandomSamp+=Texture2DSample(Tex,TexSampler,ScenePixels);\r\nRandomSamp/=5;\r\nRandomSamp-=0.5;\r\nTempAARotation*=RandomSamp;\r\nTempAADistance*=StepSize*RandomSamp;\r\n\r\n\r\nint i=0;\r\nif (DistanceSteps<1)\r\n{\r\nreturn DecodeSceneColorForMaterialNode(NewUV);\r\n}\r\nelse\r\n{\r\n//CurDistance += 0.5*StepSize;\r\nwhile ( i < (int) DistanceSteps)\r\n{\r\n\r\n//CurDistance+=StepSize;\r\nfor (int j = 0; j < (int) RadialSteps; j++)\r\n{\r\nCurOffset.x = cos(TwoPi*((TempAARotation+Substep) / RadialSteps));\r\nCurOffset.y = sin(TwoPi*((TempAARotation+Substep) / RadialSteps));\r\nCurOffset *=DistanceMask;\r\nNewUV.x = BaseUV.x + (CurOffset.x * (CurDistance+(RandomSamp*TempAADistance)));\r\nNewUV.y = BaseUV.y + (CurOffset.y * (CurDistance+(RandomSamp*TempAADistance)));\r\nCurColor += DecodeSceneColorForMaterialNode(NewUV);\r\n//CurDistance+=(StepSize+(TempAADistance))/RadialSteps;\r\nSubstep++;\r\n}\r\nCurDistance+=StepSize;\r\nSubstep+=RadialOffset;\r\ni++;\r\n}\r\nCurColor = CurColor / ((int)DistanceSteps*(int)RadialSteps);\r\nreturn CurColor;\r\n}\n*** Original source before expression upgrade ***\nfloat3 CurColor;\r\nfloat2 BaseUV = MaterialFloat2(ScreenAlignedPosition(Parameters.ScreenPosition).xy);\r\nfloat2 NewUV = BaseUV;\r\nfloat StepSize = Distance / (int) DistanceSteps;\r\nfloat CurDistance=0;\r\nfloat2 CurOffset;\r\nfloat TwoPi = 6.283185;\r\nfloat Substep;\r\nfloat2 ScenePixels=Frame.RenderTargetSize*BaseUV;\r\nScenePixels+=Frame.TemporalAAParams.r;\r\nfloat2 RandomSamp = ((uint)(ScenePixels.x) + 2 * (uint)(ScenePixels.y)) % 5;\r\nRandomSamp+=Texture2DSample(Tex,TexSampler,ScenePixels);\r\nRandomSamp/=5;\r\nRandomSamp-=0.5;\r\nTempAARotation*=RandomSamp;\r\nTempAADistance*=StepSize*RandomSamp;\r\n\r\n\r\nint i=0;\r\nif (DistanceSteps<1)\r\n{\r\nreturn DecodeSceneColorForMaterialNode(NewUV);\r\n}\r\nelse\r\n{\r\n//CurDistance += 0.5*StepSize;\r\nwhile ( i < (int) DistanceSteps)\r\n{\r\n\r\n//CurDistance+=StepSize;\r\nfor (int j = 0; j < (int) RadialSteps; j++)\r\n{\r\nCurOffset.x = cos(TwoPi*((TempAARotation+Substep) / RadialSteps));\r\nCurOffset.y = sin(TwoPi*((TempAARotation+Substep) / RadialSteps));\r\nCurOffset *=DistanceMask;\r\nNewUV.x = BaseUV.x + (CurOffset.x * (CurDistance+(RandomSamp*TempAADistance)));\r\nNewUV.y = BaseUV.y + (CurOffset.y * (CurDistance+(RandomSamp*TempAADistance)));\r\nCurColor += DecodeSceneColorForMaterialNode(NewUV);\r\n//CurDistance+=(StepSize+(TempAADistance))/RadialSteps;\r\nSubstep++;\r\n}\r\nCurDistance+=StepSize;\r\nSubstep+=RadialOffset;\r\ni++;\r\n}\r\nCurColor = CurColor / ((int)DistanceSteps*(int)RadialSteps);\r\nreturn CurColor;\r\n}"
   outputType=CMOT_Float1
   showCode=true
   additionalOutputs [{outputName=Mask, outputType=CMOT_Float1}]
   inputs [
    {inputName=Distance, input {expression=&self:MaterialExpressionNamedRerouteUsage_51}}
    {
     inputName=DistanceSteps
     input {expression=&self:MaterialExpressionNamedRerouteUsage_56}
    }
    {
     inputName=RadialSteps
     input {expression=&self:MaterialExpressionNamedRerouteUsage_57}
    }
    {
     inputName=TempAARotation
     input {expression=&self:MaterialExpressionNamedRerouteUsage_66}
    }
    {
     inputName=DistanceMask
     input {expression=&self:MaterialExpressionNamedRerouteUsage_67}
    }
    {inputName=Tex, input {expression=&self:MaterialExpressionNamedRerouteUsage_60}}
    {
     inputName=TempAADistance
     input {expression=&self:MaterialExpressionNamedRerouteUsage_68}
    }
    {
     inputName=RadialOffset
     input {expression=&self:MaterialExpressionNamedRerouteUsage_69}
    }
    {inputName=DivideF, input {expression=&self:MaterialExpressionNamedRerouteUsage_61}}
    {inputName=PDepth, input {expression=&self:MaterialExpressionNamedRerouteUsage_65}}
    {
     inputName=PixelDepthOffset
     input {expression=&self:MaterialExpressionNamedRerouteUsage_63}
    }
   ]
  }
  n17 = NamedRerouteUsage()
  n18 = Subtract(
   A=Saturate(None=StaticSwitch(True=@n15.return, False=@n16.return, Value=@n17))
   B=Saturate(None=StaticSwitch(True=@n15.Mask, False=@n16.Mask, Value=@n17))
  )
  n19 = Fresnel(ExponentIn=ScalarParameter() {defaultValue=1, group=SoftGlow, parameterName=FresnelFalloff_Exp, sortPriority=31}) {baseReflectFraction=0, exponent=1}
  n20 = Saturate(None=Power(Base=OneMinus(None=@n19), Exp=ScalarParameter() {defaultValue=1, group=SoftGlow, parameterName=FresnelFalloff_Power}))
  n21 = ScalarParameter() {group=SoftGlow, parameterName=FresnelFalloff_Intensity, sortPriority=30}
  NamedRerouteDeclaration(None=Multiply(A=Saturate(None=@n18), B=LinearInterpolate(B=@n20, Alpha=@n21) {constA=1})) {name=DepthMethod_Opacity, nodeColor=<rgba 0 0.8359 1 1>}
  n22 = StaticBoolParameter() {group=SoftGlow, parameterName=UseCustomDepth, sortPriority=0}
  n23 = StaticBool()
  NamedRerouteDeclaration(None=ShadingPathSwitch(Default=@n22, Deferred=@n22, Forward=@n23, Mobile=@n23)) {name=bUseCustomDepth, nodeColor=<rgba 1 0.7734 0 1>}
  NamedRerouteDeclaration(None=StaticBoolParameter() {group=SoftGlow, parameterName=UseFresnelGlow, sortPriority=0}) {name=bUseFresnel, nodeColor=<rgba 0.0781 1 0 1>}
  n24 = ActorPositionWS()
  n25 = VectorParameter() {group=GlowPulse, parameterName=Glow Pulse Origin + Size, defaultValue=<rgba 0 0 0 1000>}
  n26 = Length(None=Subtract(A=Add(A=@n24, B=@n25.RGB), B=WorldPosition().XYZ)) {is=class MaterialExpressionLength}
  n27 = ScalarParameter() {defaultValue=-0.7, group=GlowPulse, parameterName=Glow Pulse Speed}
  n28 = Time()
  n29 = Sine(None=Add(A=Divide(A=@n26, B=@n25.A), B=Multiply(A=@n27, B=@n28)))
  NamedRerouteDeclaration(None=Power(Base=ConstantBiasScale(None=@n29), Exp=ScalarParameter() {defaultValue=0.7, group=GlowPulse, parameterName=Glow Pulse Power})) {name=GlowPulse, nodeColor=<rgba 0 1 0.9297 1>}
  NamedRerouteDeclaration(None=ScalarParameter() {defaultValue=3, group=SoftGlow, parameterName=DistanceSteps, sortPriority=20}) {name=DistanceSteps, nodeColor=<rgba 0.7812 0 1 1>}
  NamedRerouteDeclaration(None=ScalarParameter() {defaultValue=2, group=SoftGlow, parameterName=RadialSteps, sortPriority=21}) {name=RadialSteps, nodeColor=<rgba 0 1 0.4609 1>}
  NamedRerouteDeclaration(None=TextureObject() {samplerType=SAMPLERTYPE_LinearGrayscale, texture=/Engine/EngineMaterials/Good64x64TilingNoiseHighFreq}) {name=Tex, nodeColor=<rgba 1 0 0.8203 1>}
  n30 = SceneTexture() {sceneTextureId=PPI_CustomDepth}
  n31 = SceneDepth()
  n32 = StaticSwitch(True=Multiply(A=ComponentMask(None=@n30.Color) {r=true}, B=@n31.R), False=@n31.R, Value=NamedRerouteUsage())
  NamedRerouteDeclaration(None=Multiply(A=Saturate(None=@n32), B=PixelDepth().R)) {name=PDepth, nodeColor=<rgba 0 1 0.8594 1>}
  NamedRerouteDeclaration(None=Constant()) {name=PixelDepthOffset, nodeColor=<rgba 0.2188 0 1 1>}
  NamedRerouteDeclaration(None=ScalarParameter() {defaultValue=10, group=SoftGlow, parameterName=MaxBlurDepth, sortPriority=22}) {name=MaxBlurDepth, nodeColor=<rgba 0.1016 1 0 1>}
  NamedRerouteDeclaration(None=Constant() {r=1}) {name=TempAARotation, nodeColor=<rgba 1 0 0.8203 1>}
  NamedRerouteDeclaration(None=Constant() {r=1}) {name=DistanceMask, nodeColor=<rgba 0 1 0.8594 1>}
  NamedRerouteDeclaration(None=Constant() {r=1}) {name=TempAADistance, nodeColor=<rgba 1 0.5391 0 1>}
  NamedRerouteDeclaration(None=Constant() {r=0.62}) {name=RadialOffset, nodeColor=<rgba 0.2188 0 1 1>}
  n33 = Length(None=ActorPositionWS() {originType=CameraRelative})
  n34 = Constant() {r=200}
  n35 = Power(Base=Divide(A=Subtract(A=@n33, B=@n34) {constB=100}) {constB=400}) {constExponent=0.5}
  NamedRerouteDeclaration(None=OneMinus(None=Saturate(None=@n35))) {name=DistanceBlending, nodeColor=<rgba 0.4766 1 0 1>}
  NamedRerouteDeclaration(None=Multiply(A=VertexNormalWS(), B=ScalarParameter() {defaultValue=5, group=SoftGlow, parameterName=WPO_Amount})) {name=WPO_OUT, nodeColor=<rgba 0 1 0.8828 1>}
  n36 = StaticSwitch(True=NamedRerouteUsage(), False=NamedRerouteUsage(), Value=NamedRerouteUsage())
  NamedRerouteDeclaration(None=Multiply(A=Power(Base=@n36, Exp=@n4), B=@n5)) {name=Opacity_OUT, nodeColor=<rgba 1 0.5625 0 1>}
  n37 = VectorParameter() {group=SoftGlow, parameterName=GlowColor, sortPriority=10, defaultValue=<rgba 4.8649 0.747 15 0.08>}
  NamedRerouteDeclaration(None=MaterialFunctionCall(Emissive=@n37.RGB) {materialFunction=&MF_FX_EmissiveAdjustments}.Result) {name=Emissive_OUT, nodeColor=<rgba 0.2422 0 1 1>}
  MaterialExpressionNamedRerouteUsage_31 = NamedRerouteUsage()
  MaterialExpressionNamedRerouteUsage_50 = NamedRerouteUsage()
  MaterialExpressionNamedRerouteUsage_29 = NamedRerouteUsage()
  MaterialExpressionNamedRerouteUsage_28 = NamedRerouteUsage()
  n38 = Custom!(
   DistanceMask=@MaterialExpressionNamedRerouteUsage_31
   DivideF=@MaterialExpressionNamedRerouteUsage_50
   PDepth=@MaterialExpressionNamedRerouteUsage_29
   PixelDepthOffset=@MaterialExpressionNamedRerouteUsage_28
  ) {
   code="float2 baseUV = MaterialFloat2(ScreenAlignedPosition(Parameters.ScreenPosition).xy);\r\nfloat tmpdepth2 = CalcSceneDepth(baseUV)-PDepth+PixelDepthOffset;\r\nfloat mask = (min(1-saturate(max(0,(tmpdepth2))/(DivideF+PixelDepthOffset))-0.99,0)+0.99)*saturate(sign(tmpdepth2));\r\nreturn mask;"
   desc="\n*** Original source before expression upgrade ***\nfloat3 CurColor;\r\nfloat2 BaseUV = MaterialFloat2(ScreenAlignedPosition(Parameters.ScreenPosition).xy);\r\nfloat2 NewUV = BaseUV;\r\nfloat StepSize = Distance / (int) DistanceSteps;\r\nfloat CurDistance=0;\r\nfloat2 CurOffset;\r\nfloat TwoPi = 6.283185;\r\nfloat Substep;\r\nfloat2 ScenePixels=View.RenderTargetSize*BaseUV;\r\nScenePixels+=View.TemporalAAParams.r;\r\nfloat2 RandomSamp = ((uint)(ScenePixels.x) + 2 * (uint)(ScenePixels.y)) % 5;\r\nRandomSamp+=Texture2DSample(Tex,TexSampler,ScenePixels);\r\nRandomSamp/=5;\r\nRandomSamp-=0.5;\r\nTempAARotation*=RandomSamp;\r\nTempAADistance*=StepSize*RandomSamp;\r\n\r\n\r\nint i=0;\r\nif (DistanceSteps<1)\r\n{\r\nreturn DecodeSceneColorForMaterialNode(NewUV);\r\n}\r\nelse\r\n{\r\n//CurDistance += 0.5*StepSize;\r\nwhile ( i < (int) DistanceSteps)\r\n{\r\n\r\n//CurDistance+=StepSize;\r\nfor (int j = 0; j < (int) RadialSteps; j++)\r\n{\r\nCurOffset.x = cos(TwoPi*((TempAARotation+Substep) / RadialSteps));\r\nCurOffset.y = sin(TwoPi*((TempAARotation+Substep) / RadialSteps));\r\nCurOffset *=DistanceMask;\r\nNewUV.x = BaseUV.x + (CurOffset.x * (CurDistance+(RandomSamp*TempAADistance)));\r\nNewUV.y = BaseUV.y + (CurOffset.y * (CurDistance+(RandomSamp*TempAADistance)));\r\nCurColor += DecodeSceneColorForMaterialNode(NewUV);\r\n//CurDistance+=(StepSize+(TempAADistance))/RadialSteps;\r\nSubstep++;\r\n}\r\nCurDistance+=StepSize;\r\nSubstep+=RadialOffset;\r\ni++;\r\n}\r\nCurColor = CurColor / ((int)DistanceSteps*(int)RadialSteps);\r\nreturn CurColor;\r\n}\n*** Original source before expression upgrade ***\nfloat3 CurColor;\r\nfloat2 BaseUV = MaterialFloat2(ScreenAlignedPosition(Parameters.ScreenPosition).xy);\r\nfloat2 NewUV = BaseUV;\r\nfloat StepSize = Distance / (int) DistanceSteps;\r\nfloat CurDistance=0;\r\nfloat2 CurOffset;\r\nfloat TwoPi = 6.283185;\r\nfloat Substep;\r\nfloat2 ScenePixels=Frame.RenderTargetSize*BaseUV;\r\nScenePixels+=Frame.TemporalAAParams.r;\r\nfloat2 RandomSamp = ((uint)(ScenePixels.x) + 2 * (uint)(ScenePixels.y)) % 5;\r\nRandomSamp+=Texture2DSample(Tex,TexSampler,ScenePixels);\r\nRandomSamp/=5;\r\nRandomSamp-=0.5;\r\nTempAARotation*=RandomSamp;\r\nTempAADistance*=StepSize*RandomSamp;\r\n\r\n\r\nint i=0;\r\nif (DistanceSteps<1)\r\n{\r\nreturn DecodeSceneColorForMaterialNode(NewUV);\r\n}\r\nelse\r\n{\r\n//CurDistance += 0.5*StepSize;\r\nwhile ( i < (int) DistanceSteps)\r\n{\r\n\r\n//CurDistance+=StepSize;\r\nfor (int j = 0; j < (int) RadialSteps; j++)\r\n{\r\nCurOffset.x = cos(TwoPi*((TempAARotation+Substep) / RadialSteps));\r\nCurOffset.y = sin(TwoPi*((TempAARotation+Substep) / RadialSteps));\r\nCurOffset *=DistanceMask;\r\nNewUV.x = BaseUV.x + (CurOffset.x * (CurDistance+(RandomSamp*TempAADistance)));\r\nNewUV.y = BaseUV.y + (CurOffset.y * (CurDistance+(RandomSamp*TempAADistance)));\r\nCurColor += DecodeSceneColorForMaterialNode(NewUV);\r\n//CurDistance+=(StepSize+(TempAADistance))/RadialSteps;\r\nSubstep++;\r\n}\r\nCurDistance+=StepSize;\r\nSubstep+=RadialOffset;\r\ni++;\r\n}\r\nCurColor = CurColor / ((int)DistanceSteps*(int)RadialSteps);\r\nreturn CurColor;\r\n}"
   outputType=CMOT_Float1
   showCode=true
   inputs [
    {
     inputName=DistanceMask
     input {expression=&self:MaterialExpressionNamedRerouteUsage_31}
    }
    {inputName=DivideF, input {expression=&self:MaterialExpressionNamedRerouteUsage_50}}
    {inputName=PDepth, input {expression=&self:MaterialExpressionNamedRerouteUsage_29}}
    {
     inputName=PixelDepthOffset
     input {expression=&self:MaterialExpressionNamedRerouteUsage_28}
    }
   ]
  }
  NamedRerouteDeclaration(None=@n38) {name=DepthMask, nodeColor=<rgba 0.4062 0 1 1>}
  StaticSwitchParameter(
   True=SetMaterialAttributes(EmissiveColor=NamedRerouteUsage(), Opacity=Constant() {r=1}, WorldPositionOffset=NamedRerouteUsage()) {attributeSetTypes [#1, #2, #3]}
   False=@n3
  ) {group=SoftGlow, parameterName=DEBUG, sortPriority=-1}
 ]
}
