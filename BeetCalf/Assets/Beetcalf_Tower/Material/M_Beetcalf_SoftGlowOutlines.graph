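# Translucent, unlit, two-sided material that by its name draws soft glowing outlines for the Beetcalf Tower asset.
# It uses material attributes, with its single node being a call to the MF_SoftGlowOutlines material function that holds the actual effect.
# It is flagged for use on skeletal meshes and with mesh deformers.
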Material /BeetCalf/Assets/Beetcalf_Tower/Material/M_Beetcalf_SoftGlowOutlines
graph {
 properties {
  bUseMaterialAttributes=true
  bUsedWithMeshDeformer=true
  bUsedWithSkeletalMesh=true
  blendMode=BLEND_Translucent
  pixelDepthOffsetMode=PDOM_Legacy
  shadingModel=MSM_Unlit
  twoSided=true
  usedShadingModels=Unlit
 }
 nodes [MaterialFunctionCall() {materialFunction=&/MF_SoftGlowOutlines}]
}
